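# Looping ambient sound cue that plays intermittent SprocketPoppy groans as positional outdoor one-shots, apparently for the prop's animated face.
# Each loop pass waits a random few seconds, then varies pitch and volume on a randomly picked groan.
# A quality-level node gives the first two levels all fourteen variations and the last two only the first seven.
# Face-specific attenuation and global and local concurrency settings govern falloff and overlapping instances.
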SoundCue /Hermes/Props/SprocketPoppy/Materials/FaceAnim/SC_SprocketPoppy_Face_Loop
root=Looping_0
nodes [
 {id=Delay_0, children [Modulator_0], properties {delayMax=12, delayMin=3}}
 {id=Looping_0, children [Delay_0]}
 {id=Modulator_0, children [QualityLevel_0], properties {pitchMax=1.25, pitchMin=0.85, volumeMax=1, volumeMin=0.6}}
 {id=QualityLevel_0, children [Random_1,Random_1,Random_2,Random_2]}
 {
  id=Random_1
  children [
   WavePlayer_7
   WavePlayer_8
   WavePlayer_9
   WavePlayer_10
   WavePlayer_11
   WavePlayer_12
   WavePlayer_13
   WavePlayer_14
   WavePlayer_15
   WavePlayer_16
   WavePlayer_17
   WavePlayer_18
   WavePlayer_19
   WavePlayer_20
  ]
  properties {bShouldExcludeFromBranchCulling=true, weights [1,1,1,1,1,1,1,1,1,1,1,1,1,1]}
 }
 {
  id=Random_2
  children [WavePlayer_21, WavePlayer_22, WavePlayer_23, WavePlayer_24, WavePlayer_25, WavePlayer_26, WavePlayer_27]
  properties {bShouldExcludeFromBranchCulling=true, weights [1,1,1,1,1,1,1]}
 }
 {id=WavePlayer_10, properties {soundWaveAssetPtr=&/Spot_SprocketPoppy_Groan_04}}
 {id=WavePlayer_11, properties {soundWaveAssetPtr=&/Spot_SprocketPoppy_Groan_05}}
 {id=WavePlayer_12, properties {soundWaveAssetPtr=&/Spot_SprocketPoppy_Groan_06}}
 {id=WavePlayer_13, properties {soundWaveAssetPtr=&/Spot_SprocketPoppy_Groan_07}}
 {id=WavePlayer_14, properties {soundWaveAssetPtr=&/Spot_SprocketPoppy_Groan_08}}
 {id=WavePlayer_15, properties {soundWaveAssetPtr=&/Spot_SprocketPoppy_Groan_09}}
 {id=WavePlayer_16, properties {soundWaveAssetPtr=&/Spot_SprocketPoppy_Groan_10}}
 {id=WavePlayer_17, properties {soundWaveAssetPtr=&/Spot_SprocketPoppy_Groan_11}}
 {id=WavePlayer_18, properties {soundWaveAssetPtr=&/Spot_SprocketPoppy_Groan_12}}
 {id=WavePlayer_19, properties {soundWaveAssetPtr=&/Spot_SprocketPoppy_Groan_13}}
 {id=WavePlayer_20, properties {soundWaveAssetPtr=&/Spot_SprocketPoppy_Groan_14}}
 {id=WavePlayer_21, properties {soundWaveAssetPtr=&/Spot_SprocketPoppy_Groan_01}}
 {id=WavePlayer_22, properties {soundWaveAssetPtr=&/Spot_SprocketPoppy_Groan_02}}
 {id=WavePlayer_23, properties {soundWaveAssetPtr=&/Spot_SprocketPoppy_Groan_03}}
 {id=WavePlayer_24, properties {soundWaveAssetPtr=&Spot_SprocketPoppy_Groan_04}}
 {id=WavePlayer_25, properties {soundWaveAssetPtr=&Spot_SprocketPoppy_Groan_05}}
 {id=WavePlayer_26, properties {soundWaveAssetPtr=&Spot_SprocketPoppy_Groan_06}}
 {id=WavePlayer_27, properties {soundWaveAssetPtr=&Spot_SprocketPoppy_Groan_07}}
 {id=WavePlayer_7, properties {soundWaveAssetPtr=&Spot_SprocketPoppy_Groan_01}}
 {id=WavePlayer_8, properties {soundWaveAssetPtr=&Spot_SprocketPoppy_Groan_02}}
 {id=WavePlayer_9, properties {soundWaveAssetPtr=&Spot_SprocketPoppy_Groan_03}}
]
properties {
 attenuationSettings=&/ATT_SprocketPoppy_Face
 maxDistance=2600
 soundClassObject=&Fort_Ambient_OneShots_Outdoors
 volumeMultiplier=1
 concurrencySet [
  &/CONC_SprocketPoppy_Face_Global
  &/CONC_SprocketPoppy_Face_Local
 ]
}
